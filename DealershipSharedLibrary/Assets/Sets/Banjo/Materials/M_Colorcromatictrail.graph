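# Translucent unlit Niagara material for a colour trail in the Banjo set, usable on ribbons, sprites and mesh particles.
# Two panning, noise-distorted ribbon-twist textures blend by channel or product, tinted by particle colour into emissive.
# A distorted scrolling trail mask and cell-noise holes, hardened with steps, shape opacity and blacken parts of the emissive.
# Opacity also uses a noisy smoothstep dissolve along U, particle alpha and dynamic-parameter mask intensity.
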Material /DealershipSharedLibrary/Assets/Sets/Banjo/Materials/M_Colorcromatictrail
graph {
 properties {
  bAutomaticallySetUsageInEditor=false
  bUsedWithNiagaraMeshParticles=true
  bUsedWithNiagaraRibbons=true
  bUsedWithNiagaraSprites=true
  blendMode=BLEND_Translucent
  shadingModel=MSM_Unlit
  usedShadingModels=Unlit
 }
 nodes [
  n1 = ScalarParameter() {defaultValue=1, group=MainTexture, parameterName=TrailOffset}
  n2 = ScalarParameter() {defaultValue=1, group=MainTexture, parameterName=TilingU}
  n3 = ScalarParameter() {defaultValue=1, group=MainTexture, parameterName=TilingV}
  n4 = Add(A=AppendVector(A=@n1, B=@n1), B=Multiply(A=TextureCoordinate() {uTiling=3, vTiling=0.5}, B=AppendVector(A=@n2, B=@n3)))
  n5 = AppendVector(A=@n1, B=Constant() {r=-0.08})
  n6 = VectorParameter() {group=MainTexture, parameterName="MainDistortTexture - Tiling (R,G) / Speed (B,A)", defaultValue=<rgba 0.2 1 -0.01 0.001>}
  n7 = Multiply(A=TextureCoordinate(), B=AppendVector(A=@n6.R, B=@n6.R))
  n8 = TextureSampleParameter2D(UVs=Panner(Coordinate=Add(A=@n5, B=@n7), Speed=AppendVector(A=@n6.B, B=@n6.B)) {speedX=-0.01, speedY=0.001}) {
   group=MainTexture
   parameterName=MainDistortTexture
   texture=/DealershipSharedLibrary/FX/Textures/Noise/T_VCVFX_Boost_alpine_infinite_Noise_Heightmap_Alpha
  }
  n9 = Multiply(A=ComponentMask(None=@n8.RGB) {g=true, r=true}, B=ScalarParameter() {group=MainTexture, parameterName=MainTexDistortionStrength})
  n10 = TextureSampleParameter2D(UVs=Panner(Coordinate=Add(A=@n4, B=@n9), Speed=NamedRerouteUsage()) {speedX=-0.6, speedY=-0.45}) {group=MainTexture, parameterName=SecondaryTexture, texture=/DealershipSharedLibrary/Assets/Sets/Banjo/Textures/T_BanjoRibbonTwist_D}
  n11 = TextureSampleParameter2D(UVs=Panner(Coordinate=Add(A=@n4, B=@n9), Speed=NamedRerouteUsage()) {speedX=-0.3, speedY=0.35}) {group=MainTexture, parameterName=MainTexture, texture=&T_BanjoRibbonTwist_D}
  n12 = Round(None=Saturate(None=ScalarParameter() {group=Texture Blend, parameterName="Texture to use for Blend channel (0=Main / 1=Secondary)"}))
  n13 = DotProduct(
   A=VectorParameter() {group=Texture Blend, parameterName="BlendChannelSelector (R/G/B/A)"}.RGB
   B=LinearInterpolate(A=@n11.RGBA, B=@n10.RGBA, Alpha=@n12)
  )
  n14 = LinearInterpolate(A=@n10.RGB, B=@n11.RGB, Alpha=@n13)
  n15 = LinearInterpolate(A=@n11.RGB, B=@n10.RGB, Alpha=@n13)
  n16 = Multiply(A=Power(Base=@n10.R) {constExponent=4}, B=Power(Base=@n11.R) {constExponent=6})
  n17 = Saturate(None=ScalarParameter() {defaultValue=1, group=Texture Blend, parameterName="Blend Method - 0=SpecificChannel / 1=MultipliedTextures"})
  n18 = Multiply(
   A=LinearInterpolate(A=LinearInterpolate(A=@n14, B=@n15, Alpha=@n12), B=LinearInterpolate(A=@n10.RGB, B=@n11.RGB, Alpha=@n16), Alpha=Round(None=@n17))
   B=ScalarParameter() {defaultValue=0.65, group=MainTexture, parameterName=Luminosity}
  ) {constB=0.75}
  MaterialExpressionVertexColor_0 = ParticleColor()
  n19 = VectorParameter() {
   group=Holes
   parameterName="Holes - Tiling (R,G) / Speed(B,A)"
   channelNames {a=Holes_Speed Y, b=Holes_Speed X, g=Holes_Tiling Y, r=Holes_Tiling X}
   defaultValue=<rgba 0.6 0.6 -0.25 0.1>
  }
  n20 = Multiply(A=TextureCoordinate(), B=AppendVector(A=@n19.Holes_Tiling X, B=@n19.Holes_Tiling X))
  n21 = AppendVector(A=@n19.Holes_Speed X, B=@n19.Holes_Speed X)
  n22 = TextureCoordinate()
  n23 = VectorParameter() {
   group=Holes
   parameterName="Holes Distortion - Tiling (R,G) / Speed (B,A)"
   channelNames {a=HolesDistortion_Speed Y, b=HolesDistortion_Speed X, g=HolesDistortion_Tiling Y, r=HolesDistortion_Tiling X}
   defaultValue=<rgba 0.75 0.75 -0.1 -0.05>
  }
  n24 = AppendVector(A=@n23.HolesDistortion_Tiling X, B=@n23.HolesDistortion_Tiling X)
  n25 = TextureSampleParameter2D(UVs=Panner(Coordinate=Multiply(A=@n22, B=@n24), Speed=AppendVector(A=@n23.HolesDistortion_Speed X, B=@n23.HolesDistortion_Speed X)) {speedX=-0.1, speedY=-0.05}) {
   group=Holes
   parameterName=Holes Distort Texture
   samplerType=SAMPLERTYPE_Normal
   texture=/DealershipSharedLibrary/FX/Textures/Noise/T_VCVFX_CloudNoise_N
  }
  n26 = ScalarParameter() {defaultValue=0.025, group=Holes, parameterName=HoleDistort}
  n27 = TextureSampleParameter2D(UVs=Add(A=Panner(Coordinate=@n20, Speed=@n21) {speedX=-0.25, speedY=0.1}, B=Multiply(A=@n25.R, B=@n26))) {group=Holes, parameterName=Holes Texture, texture=/DealershipSharedLibrary/FX/Textures/Noise/T_VCVFX_CellNoise01_A}
  n28 = Power(Base=@n27.R) {constExponent=50}
  n29 = Constant() {r=2}
  n30 = VectorParameter() {
   group=Mask
   parameterName="MaskTexture - Tiling (R,G) / Speed (B,A)"
   channelNames {a=MaskTexture_Speed Y, b=MaskTexture_Speed X, g=MaskTexture_Tiling Y, r=MaskTexture_Tiling X}
   defaultValue=<rgba 4 1 -2.5 0>
  }
  n31 = Multiply(A=TextureCoordinate(), B=AppendVector(A=@n30.MaskTexture_Tiling X, B=@n30.MaskTexture_Tiling X))
  n32 = VectorParameter() {
   group=Mask
   parameterName="MaskDistortionTexture - Tiling (R,G) / Speed (B,A)"
   channelNames {a=Speed Y, b=Speed X, g=Tiling Y, r=Tiling X}
   defaultValue=<rgba 1 1 -0.1 0.025>
  }
  n33 = Multiply(A=TextureCoordinate(), B=AppendVector(A=@n32.Tiling X, B=@n32.Tiling X))
  n34 = AppendVector(A=@n32.Speed X, B=@n32.Speed X)
  n35 = Multiply(
   A=TextureSampleParameter2D(UVs=Panner(Coordinate=@n33, Speed=@n34) {speedX=-0.1, speedY=-0.05}) {group=Mask, parameterName=MaskDistortionTexture, texture=/DealershipSharedLibrary/FX/Textures/Noise/T_VCVFX_clouds_noise_A}.R
   B=ScalarParameter() {group=Mask, parameterName=MaskTexture_DistortionStrength}
  )
  n36 = TextureSampleParameter2D(UVs=Panner(Coordinate=Add(A=@n31, B=@n35), Speed=AppendVector(A=@n30.MaskTexture_Speed X, B=@n30.MaskTexture_Speed X)) {speedX=-2.5}) {group=Mask, parameterName=TrailTexture, texture=/DealershipSharedLibrary/Assets/Sets/Banjo/Textures/T_BanjoRTwist_RGB}
  MaterialExpressionConstantClamp_1 = Saturate(None=Power(Base=Multiply(A=@n36.R, B=@n29), Exp=ScalarParameter() {defaultValue=1, group=Mask, parameterName=ThicknessBlack}) {constExponent=0.5})
  n37 = Saturate(None=Multiply(A=Multiply(A=@MaterialExpressionConstantClamp_1, B=@n28)) {constB=5})
  n38 = Step(X=@n37) {constY=0.0001}
  n39 = Subtract(A=@n38, B=Step(X=@n37) {constY=0.5})
  n40 = Subtract(A=@n36.G, B=@MaterialExpressionConstantClamp_1)
  n41 = Step(X=Multiply(A=Add(A=@n40, B=@n39)) {constB=10}) {constY=1}
  n42 = ScalarParameter() {defaultValue=1, group=Holes, parameterName=Holes?}
  n43 = LinearInterpolate(A=@MaterialExpressionConstantClamp_1, B=@n41, Alpha=@n42)
  n44 = OneMinus(None=@n43)
  MaterialFunctionCall(Emissive=LinearInterpolate(A=Multiply(A=@n18, B=@MaterialExpressionVertexColor_0.RGB), Alpha=LinearInterpolate(A=@n44, B=@n43, Alpha=@n42)) {constB=0}) {materialFunction=&MF_FX_EmissiveAdjustments}
  n45 = Multiply(A=@n36.G, B=OneMinus(None=@n38))
  n46 = Constant() {r=3}
  n47 = ComponentMask(None=@n27.RGB) {g=true, r=true}
  n48 = Constant()
  n49 = Panner(Coordinate=Add(A=TextureCoordinate(), B=Multiply(A=@n47, B=@n48))) {speedX=-0.5, speedY=-0.5}
  n50 = ComponentMask(None=TextureCoordinate()) {r=true}
  n51 = Add(
   A=ConstantBiasScale(None=TextureSample(UVs=@n49) {samplerType=SAMPLERTYPE_Normal, texture=&T_VCVFX_CloudNoise_N}.R)
   B=Multiply(A=OneMinus(None=@n50), B=ScalarParameter() {defaultValue=2.5, group=Mask, parameterName=DissolveTrail})
  )
  n52 = ScalarParameter() {defaultValue=0.85, group=Mask, parameterName=SmoothstepMin}
  n53 = ScalarParameter() {defaultValue=1, group=Mask, parameterName=SmoothstepMax}
  n54 = Saturate(None=Divide(A=Subtract(A=@n51, B=@n52), B=Subtract(A=@n53, B=@n52)))
  n55 = Multiply(A=@n54, B=Constant() {r=2})
  n56 = Multiply(A=Multiply(A=Subtract(A=@n46, B=@n55), B=@n54), B=@n54)
  n57 = AppendVector(A=AppendVector(A=@n56, B=@n56), B=@n56)
  n58 = Constant() {r=10}
  MaterialExpressionConstantClamp_0 = Saturate(None=Multiply(A=@n11.A, B=Power(Base=@n57, Exp=@n58)))
  n59 = Multiply(A=Multiply(A=@MaterialExpressionConstantClamp_0, B=LinearInterpolate(A=@n36.G, B=@n45, Alpha=@n42))) {constB=50}
  n60 = Subtract(A=@MaterialExpressionConstantClamp_1, B=@n38)
  n61 = Multiply(A=@MaterialExpressionConstantClamp_0, B=LinearInterpolate(A=@n40, B=Add(A=@n41, B=@n60), Alpha=@n42))
  n62 = DynamicParameter() {defaultValue=<rgba 1 1 1 3>, paramNames [TrailOffset, Param2, mask_x offset, mask_intensity]}
  Multiply(
   A=Add(A=Multiply(A=@MaterialExpressionVertexColor_0.A, B=@n61), B=Step(X=@n59) {constY=1})
   B=Multiply(A=Multiply(A=@n50, B=@n62.mask_intensity), B=@n36.G)
  )
  n63 = VectorParameter() {
   group=MainTexture
   parameterName=Textures Panning Speed
   sortPriority=1
   channelNames {a=SecondaryTexture_Speed Y, b=SecondayTexture_Speed X, g=MainTexture_Speed Y, r=MainTexture_Speed X}
   defaultValue=<rgba -0.3 0.35 -0.6 -0.45>
  }
  NamedRerouteDeclaration(None=AppendVector(A=@n63.MainTexture_Speed X, B=@n63.MainTexture_Speed X)) {name=MainTextureSpeed, nodeColor=<rgba 0.9688 1 0 1>}
  NamedRerouteDeclaration(None=AppendVector(A=@n63.SecondayTexture_Speed X, B=@n63.SecondayTexture_Speed X)) {name=SecondaryTextureSpeed, nodeColor=<rgba 0.7109 0 1 1>}
 ]
}
